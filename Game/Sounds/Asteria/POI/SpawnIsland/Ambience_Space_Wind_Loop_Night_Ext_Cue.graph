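# Looping outdoor night space-wind ambience cue for the Asteria SpawnIsland POI, using ambient outdoor sound class and attenuation.
# A quality-level node routes its first two slots to a quad wave and its last two to a stereo wave.
# Each looping wave passes through two nested endless envelopers: a 30-second pitch drift inside a 60-second volume swell, adding slow variation to the static loop.
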SoundCue /Game/Sounds/Asteria/POI/SpawnIsland/Ambience_Space_Wind_Loop_Night_Ext_Cue
root=QualityLevel_1
nodes [
 {
  id=Enveloper_0
  children [Enveloper_7]
  properties {
   bLoop=true
   bLoopIndefinitely=true
   loopEnd=60
   pitchCurve {editorCurveData {keys [{time=0}]}}
   volumeCurve {editorCurveData {keys [
      {arriveTangent=-0.01, interpMode=RCIM_Cubic, time=0, value=1}
      {arriveTangent=0.0006, interpMode=RCIM_Cubic, leaveTangent=0.0006, tangentMode=RCTM_User, time=15, value=0.8}
      {arriveTangent=-0.0006, interpMode=RCIM_Cubic, leaveTangent=-0.0006, tangentMode=RCTM_User, time=25, value=1}
      {arriveTangent=0.0002, interpMode=RCIM_Cubic, leaveTangent=0.0002, tangentMode=RCTM_User, time=45, value=0.6}
      <at 60 1>
     ]}}
  }
 }
 {
  id=Enveloper_4
  children [Enveloper_5]
  properties {
   bLoop=true
   bLoopIndefinitely=true
   loopEnd=60
   pitchCurve {editorCurveData {keys [{time=0}]}}
   volumeCurve {editorCurveData {keys [
      {arriveTangent=-0.01, interpMode=RCIM_Cubic, time=0, value=1}
      {arriveTangent=0.0006, interpMode=RCIM_Cubic, leaveTangent=0.0006, tangentMode=RCTM_User, time=15, value=0.8}
      {arriveTangent=-0.0006, interpMode=RCIM_Cubic, leaveTangent=-0.0006, tangentMode=RCTM_User, time=25, value=1}
      {arriveTangent=0.0002, interpMode=RCIM_Cubic, leaveTangent=0.0002, tangentMode=RCTM_User, time=45, value=0.6}
      <at 60 1>
     ]}}
  }
 }
 {
  id=Enveloper_5
  children [WavePlayer_0]
  properties {
   bLoop=true
   bLoopIndefinitely=true
   loopEnd=30
   pitchCurve {editorCurveData {keys [{interpMode=RCIM_Cubic, time=0}, <at 15.9468 0.9181 RCIM_Cubic>, {interpMode=RCIM_Cubic, time=30}]}}
   volumeCurve {editorCurveData {keys [<at 0 1>]}}
  }
 }
 {
  id=Enveloper_7
  children [WavePlayer_6]
  properties {
   bLoop=true
   bLoopIndefinitely=true
   loopEnd=30
   pitchCurve {editorCurveData {keys [{interpMode=RCIM_Cubic, time=0}, <at 15.9468 0.9181 RCIM_Cubic>, {interpMode=RCIM_Cubic, time=30}]}}
   volumeCurve {editorCurveData {keys [<at 0 1>]}}
  }
 }
 {id=QualityLevel_1, children [Enveloper_4,Enveloper_4,Enveloper_0,Enveloper_0]}
 {id=WavePlayer_0, properties {bLooping=true, soundWaveAssetPtr=&/Ambience_Space_Wind_Night_Quad}}
 {id=WavePlayer_6, properties {bLooping=true, soundWaveAssetPtr=&/Ambience_Space_Wind_Night_Stereo}}
]
properties {
 attenuationSettings=&Fort_Ambient_Athena_Att
 maxDistance=2097152
 soundClassObject=&Fort_Ambient_Outdoors_Athena
 volumeMultiplier=0.351
}
